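# Conversation database for the CasinoReaper NPC (Cassidy Quinn), defining the greeting and branches offered when a player talks to the NPC.
# The entry point feeds one Speech node that selects a line by requirement and priority: first-ever meeting, match conversation count, or reactions to specific player outfits.
# Link nodes then jump by entry tag to shared quests, basic services, character dialogue, Sniper expertise, and Season 34 sniper and carjacking conversations.
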ConversationDatabase /NPCLibrary/NPCs/CasinoReaper/NPCLibrary_Conversation_CasinoReaper
entry_points [{tag=Conversation.Entry.CasinoReaper, destinations [#1]}]
nodes [
 ConversationEntryPointNode_7 {
  id=#1
  class=ConversationEntryPointNode
  outputs [#2]
  properties {entryTag {tagName=Conversation.Entry.CasinoReaper}}
 }
 Speech_C_4 {
  id=#2
  class=Speech_C
  outputs [
   #3
   #4
   #5
   #6
   #7
   #8
   #9
  ]
  properties {
   nodeName=""
   generalConfig {
    defaultMessage=I'm all in.
    contextualMessages [
     {
      message="Cassidy Quinn, at your disservice."
      contextRequirements [{requirement {class=FortControllerRequirement_IsFirstEverConversationWithNPC}}]
      priority {curve {rowName=Tandem.Speech.Priority.FirstEverInteraction}}
     }
     {
      message=I'm all in.
      contextRequirements [{requirement {class=FortControllerRequirement_CheckNPCMatchConversationCount, properties {nPCUniqueIdTag {tagName=AISpawnerData.Type.Tandem.CasinoReaper}}}}]
      priority {curve {rowName=None}}
     }
     {
      message=Has fate dealt you a good hand today?
      contextRequirements [{requirement {class=FortControllerRequirement_CheckNPCMatchConversationCount, properties {nPCUniqueIdTag {tagName=AISpawnerData.Type.Tandem.CasinoReaper}}}}]
      priority {curve {rowName=None}}
     }
     {
      message="Some people call me the Quinn of Spades. You know, when I ask them to."
      contextRequirements [{requirement {class=FortControllerRequirement_CheckNPCMatchConversationCount, properties {nPCUniqueIdTag {tagName=AISpawnerData.Type.Tandem.CasinoReaper}}}}]
      priority {curve {rowName=None}}
     }
     {
      message="Whatever you're planning, deal me in."
      contextRequirements [{requirement {class=FortControllerRequirement_CheckNPCMatchConversationCount, properties {nPCUniqueIdTag {tagName=AISpawnerData.Type.Tandem.CasinoReaper}}}}]
      priority {curve {rowName=None}}
     }
     {
      message=I'll slow down when I'm good and ready for it.
      contextRequirements [{requirement {class=FortControllerRequirement_HasCID, properties {allowedCIds [/BRCosmetics/Athena/Items/Cosmetics/Characters/Character_TealMink_Pelt]}}}]
      priority {curve {rowName=Tandem.Speech.Priority.OutfitReaction}}
     }
     {
      message="You're... Midas!! THE Midas!! Oh... Oh wow...."
      contextRequirements [{requirement {class=FortControllerRequirement_HasCID, properties {allowedCIds [&Character_AuraCop_Bullet]}}}]
      priority {curve {rowName=Tandem.Speech.Priority.OutfitReaction}}
     }
     {
      message=You're not gonna catch me.
      contextRequirements [{requirement {class=FortControllerRequirement_HasCID, properties {allowedCIds [/BRCosmetics/Athena/Items/Cosmetics/Characters/Character_SherWolf_Gnash]}}}]
      priority {curve {rowName=Tandem.Speech.Priority.OutfitReaction}}
     }
    ]
   }
  }
 }
 ConversationLinkNode_6 {id=#9, class=ConversationLinkNode, properties {remoteEntryTag {tagName=Conversation.Entry.Expertise.Sniper}}}
 ConversationLinkNode_5 {
  id=#8
  class=ConversationLinkNode
  properties {remoteEntryTag {tagName=Conversation.Entry.Season34.Scenario.Carjacking}}
 }
 ConversationLinkNode_4 {
  id=#7
  class=ConversationLinkNode
  properties {remoteEntryTag {tagName=Conversation.Entry.Season34.Expertise.Sniper.Finish}}
 }
 ConversationLinkNode_3 {
  id=#6
  class=ConversationLinkNode
  properties {remoteEntryTag {tagName=Conversation.Entry.Season34.Expertise.Sniper.Introduction}}
 }
 ConversationLinkNode_2 {
  id=#5
  class=ConversationLinkNode
  properties {remoteEntryTag {tagName=Conversation.Entry.CasinoReaper.Dialogue}}
 }
 ConversationLinkNode_0 {id=#4, class=ConversationLinkNode, properties {remoteEntryTag {tagName=Conversation.Entry.BasicServices}}}
 ConversationLinkNode_1 {id=#3, class=ConversationLinkNode, properties {remoteEntryTag {tagName=Conversation.Entry.NPCLibrary.Quests}}}
]
